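# Translucent unlit two-sided Niagara ribbon and sprite material for the RacerZero pickaxe geode effect, apparently a dark-background variant.
# A rotated electric-crack texture and center-scaled sphere mask distort UVs for two opposite-panning turbulence samples, blended by the cracks' blue channel.
# Color adds the crack red and green channels to masked turbulence.
# Opacity adds turbulence to particle alpha, shapes it with sine edge falloffs, and applies depth fade except on mobile.
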Material /BRCosmetics/Weapons/FORT_Melee/Pickaxe_RacerZero/FX/M_Geode_Ribbon_DarkBkgd
graph {
 properties {
  bAutomaticallySetUsageInEditor=false
  bUsedWithNiagaraRibbons=true
  bUsedWithNiagaraSprites=true
  bUsedWithStaticMesh=false
  blendMode=BLEND_Translucent
  pixelDepthOffsetMode=PDOM_Legacy
  shadingModel=MSM_Unlit
  twoSided=true
  usedShadingModels=Unlit
 }
 nodes [
  n1 = Constant() {r=1}
  n2 = ScalarParameter() {defaultValue=1, parameterName=TileMult}
  n3 = MaterialFunctionCall(
   UVs=Multiply(A=TextureCoordinate() {coordinateIndex=1, vTiling=0.25}, B=AppendVector(A=@n1, B=@n2))
   "Rotation Angle (0-1)"=ScalarParameter() {defaultValue=0.25, parameterName=Rotation}
  ) {materialFunction=&CustomRotator}
  n4 = TextureSample(UVs=@n3.Rotated Values) {texture=/Game/Effects/Fort_Effects/Textures/Electric/T_Vert_Electric_Cracks}
  n5 = LinearInterpolate(Alpha=@n4.B) {constA=1, constB=-1}
  n6 = TextureCoordinate()
  n7 = ComponentMask(None=TextureCoordinate()) {r=true}
  n8 = AppendVector(A=Clamp(None=Power(Base=@n7) {constExponent=0.6}) {minDefault=0.1}, B=Constant() {r=1})
  n9 = SphereMask(
   A=ComponentMask(None=MaterialFunctionCall(UVs=@n6, Texture Scale=@n8) {materialFunction=&ScaleUVsByCenter}.UVs) {g=true}
   B=Constant() {r=0.5}
  ) {attenuationRadius=0.5, hardnessPercent=0}
  n10 = TextureCoordinate() {coordinateIndex=1, uTiling=1.5}
  n11 = AppendVector(A=Constant() {r=1}, B=ScalarParameter() {parameterName=TileMult})
  n12 = Add(
   A=Multiply(A=Multiply(A=@n5, B=@n9) {constA=1}, B=Constant2Vector() {r=-0.06}) {constB=0.2}
   B=MaterialFunctionCall(UVs=Multiply(A=@n10, B=@n11), "Rotation Angle (0-1)"=ScalarParameter() {defaultValue=0.25, parameterName=Rotation}) {materialFunction=&CustomRotator}.Rotated Values
  )
  n13 = Multiply(A=Time(), B=ScalarParameter() {defaultValue=1, parameterName=PanSpeed})
  n14 = Panner(Coordinate=@n12, Time=@n13) {speedX=0.1, speedY=0.05}
  n15 = Panner(Coordinate=@n12, Time=@n13) {speedX=-0.1, speedY=0.05}
  n16 = Multiply(A=LinearInterpolate(
    A=TextureSample(UVs=@n14) {texture=/Game/Global/Textures/Default/Caustics/T_Turbulence_Seamless}.R
    B=TextureSample(UVs=@n15) {texture=&T_Turbulence_Seamless}.R
    Alpha=@n4.B
   )) {constB=2}
  Add(
   A=Multiply(
    A=TextureSample(UVs=@n3.Rotated Values) {texture=&T_Vert_Electric_Cracks}.RGB
    B=Constant3Vector() {constant=<rgba 1 1 0 0>}
   )
   B=Multiply(A=Add(A=@n16) {constB=0.15}, B=Power(Base=@n9)) {constA=1}
  )
  n17 = TextureCoordinate()
  n18 = ComponentMask(None=@n17) {g=true}
  n19 = ComponentMask(None=@n17) {r=true}
  n20 = Multiply(A=Add(A=ParticleColor().A, B=@n16), B=Multiply(A=Sine(None=@n18) {period=2}, B=Sine(None=@n19) {period=2}))
  ShadingPathSwitch(
   Default=DepthFade(Opacity=@n20, FadeDistance=ScalarParameter() {defaultValue=800, parameterName=DepthFadeDIst}) {fadeDistanceDefault=500}
   Mobile=@n20
  )
 ]
}
