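# Translucent, two-sided unlit ribbon trail material for Niagara effects on the RacerZero pickaxe.
# Curve-atlas lookups over swizzlable UVs supply color, edge masks and shaping, with color passing through emissive adjustments.
# Opacity multiplies those masks by rotating screen-space noise (a constant on mobile), a particle-alpha threshold along V and a sine across U.
# Panning noise with dynamic-parameter randomness, weighted by curve channels, produces scalable world position offset wobble.
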Material /BRCosmetics/Weapons/FORT_Melee/Pickaxe_RacerZero/FX/M_RibbonTrail
graph {
 properties {
  bUsedWithNiagaraRibbons=true
  blendMode=BLEND_Translucent
  pixelDepthOffsetMode=PDOM_Legacy
  shadingModel=MSM_Unlit
  twoSided=true
  usedShadingModels=Unlit
 }
 nodes [
  n1 = TextureCoordinate()
  n2 = MaterialFunctionCall(Float2=MaterialFunctionCall(XY=@n1) {materialFunction=&Swizzle}.YX) {materialFunction=&BreakOutFloat2Components}
  n3 = LinearInterpolate(A=@n1, B=AppendVector(A=@n2.R, B=OneMinus(None=@n2.G)), Alpha=ScalarParameter() {parameterName=UVswizzle})
  n4 = ComponentMask(None=@n3) {g=true}
  n5 = MaterialFunctionCall(Float2=@n3) {materialFunction=&BreakOutFloat2Components}
  n6 = CurveAtlasRowParameter(CurveTime=@n5.R) {
   atlas=&AthenaCurveAtlas01
   curve=/Game/Gadgets/Assets/VinderTech_GliderChute/Glider_ShatterFly/FX/Curve_Ramp
   defaultValue=75
  }
  n7 = ScalarParameter() {defaultValue=-0.25, parameterName=MaskAffectColor}
  n8 = CurveAtlasRowParameter(CurveTime=Add(A=OneMinus(None=@n4), B=Multiply(A=@n6.B, B=@n7))) {
   atlas=&AthenaCurveAtlas01
   curve=/Game/Gadgets/Assets/VinderTech_GliderChute/Glider_ShatterFly/FX/Curve_Color2
   defaultValue=76
   parameterName=ParamC
  }
  MaterialFunctionCall(Emissive=@n8) {materialFunction=&MF_FX_EmissiveAdjustments}
  n9 = ComponentMask(None=@n3) {g=true}
  n10 = CurveAtlasRowParameter(CurveTime=@n9) {
   atlas=&AthenaCurveAtlas01
   curve=&Curve_Ramp
   defaultValue=75
  }
  n11 = Subtract(A=@n10.G)
  n12 = ComponentMask(None=@n3) {r=true}
  n13 = Add(A=@n9, B=Add(A=DynamicParameter() {paramNames [Random,Param2,Param3,Param4]}.Random, B=ScalarParameter() {parameterName=WPO_Random}))
  n14 = Constant2Vector() {g=64, r=2}
  n15 = TextureSample(UVs=Panner(Coordinate=Divide(A=@n13, B=@n14), Speed=Divide(B=@n14))) {samplerType=SAMPLERTYPE_LinearColor, texture=&HighContrastNoise}
  Multiply(
   A=Multiply(A=LinearInterpolate(A=@n10.R, B=@n11, Alpha=@n12) {constA=0.1}, B=Subtract(A=@n15.RGB) {constB=0.5})
   B=ScalarParameter() {defaultValue=1, parameterName=WPO_Scale}
  )
  n16 = ScreenPosition()
  n17 = Multiply(A=Rotator(Coordinate=Subtract(A=@n16.ViewportUV) {constB=0.5}), B=Constant2Vector() {g=2, r=2})
  n18 = Multiply(A=@n17) {constB=-0.75}
  n19 = LinearInterpolate(
   A=ScalarParameter() {defaultValue=0.05, parameterName=OpacityB}
   B=ScalarParameter() {defaultValue=1.25, parameterName=OpacityA}
   Alpha=Multiply(
    A=TextureSample(UVs=@n17) {samplerType=SAMPLERTYPE_LinearColor, texture=&HighContrastNoise}.R
    B=TextureSample(UVs=@n18) {samplerType=SAMPLERTYPE_LinearColor, texture=&HighContrastNoise}.G
   )
  ) {constA=0.05, constB=1.25}
  n20 = Constant() {r=0.5}
  n21 = CurveAtlasRowParameter(CurveTime=@n5.G) {
   atlas=&AthenaCurveAtlas01
   curve=&Curve_Ramp
   defaultValue=75
  }
  n22 = Multiply(A=Subtract(A=ComponentMask(None=@n3) {g=true}, B=ParticleColor().A)) {constB=2}
  n23 = ComponentMask(None=TextureCoordinate()) {r=true}
  Multiply(
   A=Multiply(A=ShadingPathSwitch(Default=@n19, Mobile=@n20) {desc=Migrated from FeatureLevelSwitch}, B=Multiply(A=@n6.B, B=@n21.A))
   B=Multiply(A=Saturate(None=@n22), B=Sine(None=@n23) {period=2})
  )
  TextureCoordinate()
 ]
}
