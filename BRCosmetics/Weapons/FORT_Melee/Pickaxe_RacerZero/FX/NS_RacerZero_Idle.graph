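# Niagara idle effect for the RacerZero melee pickaxe, layering two continuously spawned sprite smoke emitters along a thin line on the X axis.
# Backpack_Matching_SMoke spawns sparse pink HDR puffs that grow while fading in and out.
# Glider_Matching_Smoke spawns denser puffs with slight upward acceleration and drag, shifting from blue to violet as they grow and fade.
# The glider layer also animates a dynamic material parameter over particle life.
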NiagaraSystem /BRCosmetics/Weapons/FORT_Melee/Pickaxe_RacerZero/FX/NS_RacerZero_Idle
dependencies {
 usedDataInterfaces [&NiagaraDataInterfaceCurve, &NiagaraDataInterfaceColorCurve]
 usedDynamicInputs []
 usedModules [&SystemState]
}
emitters {
 Backpack_Matching_SMoke {
  emitterUpdateScript [
   &EmitterState()
   &SpawnRate(SpawnRate=2, Use Spawn Probability=false)
  ]
  particleSpawnScript [
   /Niagara/Modules/Spawn/Initialization/InitializeParticle(
    Color=&MakeLinearColorFromVectorAndFloat(
     "Vector (RGB)"=&MakeVectorFromLinearColorRGB(Color=<rgba 2 0.4928 1.7358 1>)
     "Float (Alpha)"=&UniformRangedFloat(Minimum=0.5, Maximum=1)
    )
    Lifetime=&UniformRangedFloat(Minimum=1.5, Maximum=2)
    Module.Write Color=true
    Module.Write Lifetime=true
    Module.Write Mass=true
    Module.Write Material Random=true
    Module.Write Position=false
    Module.Write Scale=false
    Module.Write SpriteRotation=true
    Module.Write SpriteSize=true
    Position=&SimulationPosition()
    Sprite Rotation=/Niagara/DynamicInputs/Angles/DegreesToNormalizedAngle(AngleInDegrees:NiagaraFloat=&UniformRangedFloat(Minimum=0, Maximum=1))
    Sprite Size=&UniformRangedVector2D(Minimum=<25 25>, Maximum=<50 25>)
   )
   &BoxLocation(Box Size=<50 0 0>, MaskSpawnGroup=false, Module.Override Seed=false, Rotate Box=false, Surface Only=false)
  ]
  particleUpdateScript [
   &ParticleState()
   &SolveForcesAndVelocity(Clamp Velocity=false, Limit Acceleration=false)
   &ScaleColor(
    Scale Alpha=&FloatFromCurve(FloatCurve=DataInterface() {propertyValues {Curve {keys [<at 0 0 RCIM_Cubic>,<at 0.5 1 RCIM_Cubic>,<at 1 0 RCIM_Cubic>]}}})
    ScaleA=true
    ScaleRGB=true
   )
   &ScaleSpriteSize(Scale Factor=&Vector2DFromFloat(Value=&FloatFromCurve(FloatCurve=DataInterface() {propertyValues {Curve {keys [<at 0 0.5 RCIM_Cubic>, {interpMode=RCIM_Cubic, leaveTangent=3.1399, time=1, value=1}]}}})))
  ]
  rendererClasses [&NiagaraSpriteRendererProperties]
  renderers [{rendererClass=&NiagaraSpriteRendererProperties}]
 }
 Glider_Matching_Smoke {
  emitterUpdateScript [
   &EmitterState()
   &SpawnRate(SpawnRate=10, Use Spawn Probability=false)
  ]
  particleSpawnScript [
   &InitializeParticle(
    Color=&MakeLinearColorFromVectorAndFloat("Vector (RGB)"=<1 1 1>, "Float (Alpha)"=&UniformRangedFloat(Minimum=0.2, Maximum=0.5))
    Lifetime=&UniformRangedFloat(Minimum=1.5, Maximum=2)
    Module.Write Color=true
    Module.Write Lifetime=true
    Module.Write Mass=true
    Module.Write Material Random=true
    Module.Write Position=false
    Module.Write Scale=false
    Module.Write SpriteRotation=false
    Module.Write SpriteSize=true
    Position=&SimulationPosition()
    Sprite Size=&Vector2DFromFloat(Value=&UniformRangedFloat(Minimum=25, Maximum=45))
   )
   &BoxLocation(Box Size=<55 0 0>, MaskSpawnGroup=false, Module.Override Seed=false, Rotate Box=false, Surface Only=false)
  ]
  particleUpdateScript [
   &ParticleState()
   &AccelerationForce(Acceleration=&UniformRangedVector(Minimum=<0 0 0>, Maximum=<0 0 5>))
   &Drag(Drag=&UniformRangedFloat(Minimum=2, Maximum=3), Use Linear Drag=true, Use Rotational Drag=false)
   &SolveForcesAndVelocity(Clamp Velocity=false, Limit Acceleration=false)
   &DynamicMaterialParameters(
    Index 0 Param 1=&FloatFromCurve(FloatCurve=DataInterface() {propertyValues {Curve {keys [{arriveTangent=-6.677, interpMode=RCIM_Cubic, leaveTangent=-6.677, tangentMode=RCTM_User, time=0, value=2}, <at 1 0 RCIM_Cubic>]}}})
    Index 0 Param 2=1
    Index 0 Param 4=0.5
    Module.Param0WriteEnabled=true
    Module.Param1WriteEnabled=true
    Module.Param2WriteEnabled=true
    Module.Param3WriteEnabled=true
   )
   &ScaleColor(
    Scale Mode=::NewEnumerator1
    Scale RGBA=&Vector4FromCurve(Vector4Curve=DataInterface() {propertyValues {
       AlphaCurve {keys [<at -0.0301 0>,<at 0.1507 0.2>,<at 1.0014 0>]}
       BlueCurve {keys [<at 0.085 3>,<at 0.7617 3>]}
       GreenCurve {keys [<at 0.085 0.5205>,<at 0.7617 0>]}
       RedCurve {keys [<at 0.085 0>,<at 0.7617 1.9783>]}
      }})
    ScaleRGBA=true
   )
   &ScaleSpriteSize(Scale Factor=&Vector2DFromFloat(Value=&FloatFromCurve(FloatCurve=DataInterface() {propertyValues {Curve {keys [<at 0 0>,<at 0.5 1>,<at 1 1>]}}})))
  ]
  rendererClasses [&NiagaraSpriteRendererProperties]
  renderers [{rendererClass=&NiagaraSpriteRendererProperties}]
 }
}
summary {
 systemName=NS_RacerZero_Idle
 emitters [
  {emitterName=Backpack_Matching_SMoke, rendererClasses [&NiagaraSpriteRendererProperties]}
  {emitterName=Glider_Matching_Smoke, rendererClasses [&NiagaraSpriteRendererProperties]}
 ]
 userVariables []
}
