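# Niagara swing effect for the RacerZero pickaxe, built from three independent emitters with short, delayed spawn windows and no user parameters.
# DarkBKGD draws a brief, low-alpha ribbon trail that tapers across the emitter loop and each particle's life.
# Fragments throws drifting, tumbling meshes that slow under drag, flash bright, then fade.
# Sparklies scatters cyan sprites that twinkle through a spiked size curve while their camera offset shifts over life.
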NiagaraSystem /BRCosmetics/Weapons/FORT_Melee/Pickaxe_RacerZero/FX/NS_RacerZero_Swing
dependencies {
 usedDataInterfaces [&NiagaraDataInterfaceCurve, &NiagaraDataInterfaceVectorCurve]
 usedDynamicInputs []
 usedModules [&SystemState]
}
emitters {
 DarkBKGD {
  emitterUpdateScript [
   &EmitterState(Life Cycle Mode=::NewEnumerator1, Loop Behavior=::NewEnumerator1, Loop Delay=0.05, Loop Duration=0.15, UseLoopDelay=true)
   &SpawnRate(SpawnRate=50, Use Spawn Probability=false)
  ]
  particleSpawnScript [&InitializeRibbon(
    Color=<rgba 500 500 500 0.1>
    Lifetime=0.2
    Module.Write Color=true
    Module.Write Lifetime=true
    Module.Write Mass=true
    Module.Write Material Random=true
    Module.Write Position=false
    Module.Write RibbonTwist=false
    Module.Write RibbonWidth=true
    Ribbon Width=&Lerp_Float(A=40, B=0, Alpha=$Emitter.NormalizedLoopAge)
   )]
  particleUpdateScript [
   &ParticleState()
   &SolveForcesAndVelocity(Clamp Velocity=false, Limit Acceleration=false)
   &ScaleColor(
    Scale Alpha=&FloatFromCurve(FloatCurve=DataInterface() {propertyValues {Curve {keys [<at 0 1>,<at 1 0>]}}})
    ScaleA=true
    ScaleRGB=true
   )
   &ScaleRibbonWidth(Ribbon Width Scale=&FloatFromCurve(FloatCurve=DataInterface() {propertyValues {Curve {keys [<at 0 1 RCIM_Cubic>, {arriveTangent=-1.8739, interpMode=RCIM_Cubic, leaveTangent=-1.8739, tangentMode=RCTM_User, time=1, value=0}]}}}))
  ]
  rendererClasses [&NiagaraRibbonRendererProperties]
  renderers [{rendererClass=&NiagaraRibbonRendererProperties}]
 }
 Fragments {
  emitterUpdateScript [
   &EmitterState(Life Cycle Mode=::NewEnumerator1, Loop Behavior=::NewEnumerator1, Loop Delay=0.1, Loop Duration=0.05, UseLoopDelay=true)
   &SpawnRate(SpawnRate=100, Use Spawn Probability=false)
  ]
  particleSpawnScript [
   /Niagara/Modules/Spawn/Initialization/InitializeParticle(
    Color=&MakeLinearColorFromVectorAndFloat(
     "Vector (RGB)"=&VectorFromFloat(Value=&UniformRangedFloat(Minimum=1, Maximum=5))
     "Float (Alpha)"=1
    )
    Lifetime=&UniformRangedFloat(Minimum=0.5, Maximum=0.8)
    Mesh Scale=&UniformRangedVector(Minimum=<0.3 0.3 0.3>, Maximum=<1.2 0.8 0.8>)
    Module.Write Color=true
    Module.Write Lifetime=true
    Module.Write Mass=true
    Module.Write Material Random=true
    Module.Write Position=false
    Module.Write Scale=true
    Module.Write SpriteRotation=false
    Module.Write SpriteSize=true
    Position=&SimulationPosition()
    Sprite Size=<5 5>
   )
   /Niagara/Modules/Spawn/Orientation/V2/InitialMeshOrientation(
    Orientation Coordinate Space:ENiagaraCoordinateSpace=::Local
    Rotation=&UniformRangedVector(Minimum=<0 0 0>, Maximum=<1 1 1>)
    Use Orientation Vector=false
    Use Rotation Vector=true
   ) {enabled=false}
   /Niagara/Modules/Spawn/Orientation/V3/InitialMeshOrientation(
    Orientation Coordinate Space=::Local
    Rotation=&UniformRangedVector(Minimum=<0 0 0>, Maximum=<1 1 1>)
    Use Orientation Vector=false
    Use Rotation Vector=true
   ) {module=InitialMeshOrientation001}
   &CameraOffset(Camera Offset Amount=25)
   &AddVelocity(Velocity=&UniformRangedVector(Minimum=<-45 -45 -75>, Maximum=<45 45 75>))
  ]
  particleUpdateScript [
   &ParticleState()
   &AccelerationForce(Acceleration=&UniformRangedVector(Minimum=<0 0 -50>, Maximum=<0 0 50>))
   &Drag(Drag=&UniformRangedFloat(Minimum=2, Maximum=4), Use Linear Drag=true, Use Rotational Drag=false)
   &SolveForcesAndVelocity(Clamp Velocity=false, Limit Acceleration=false)
   &UpdateMeshOrientation(
    Rotation Rate=&Lerp_Float(A=&UniformRangedFloat(Minimum=0.1, Maximum=0.4), B=0, Alpha=$Particles.NormalizedAge)
    Rotation Vector=<1 0 0>
   ) {module=UpdateMeshOrientation003}
   &UpdateMeshOrientation(
    Rotation Rate=&Lerp_Float(A=&UniformRangedFloat(Minimum=1.5, Maximum=2), B=0, Alpha=$Particles.NormalizedAge)
    Rotation Vector=<0 1 0>
   ) {module=UpdateMeshOrientation004}
   &UpdateMeshOrientation(
    Rotation Rate=&Lerp_Float(A=&UniformRangedFloat(Minimum=-0.1, Maximum=0.1), B=0, Alpha=$Particles.NormalizedAge)
    Rotation Vector=<0 0 1>
   ) {module=UpdateMeshOrientation005}
   &ScaleMeshSize(Scale Factor=&VectorFromCurve(VectorCurve=DataInterface() {propertyValues {
       XCurve {keys [
         {arriveTangent=5.4632, interpMode=RCIM_Cubic, leaveTangent=5.4632, tangentMode=RCTM_User, time=0, value=0}
         {arriveTangent=0.9375, interpMode=RCIM_Cubic, leaveTangent=0.9375, time=0.2, value=1}
         {arriveTangent=-1.25, interpMode=RCIM_Cubic, leaveTangent=-1.25, time=0.8, value=0.75}
         <at 1 0 RCIM_Cubic>
        ]}
       YCurve {keys [
         {arriveTangent=4.7574, interpMode=RCIM_Cubic, leaveTangent=4.7574, tangentMode=RCTM_User, time=0, value=0}
         {arriveTangent=1.0625, interpMode=RCIM_Cubic, leaveTangent=1.0625, time=0.2, value=1}
         {arriveTangent=-0.25, interpMode=RCIM_Cubic, leaveTangent=-0.25, time=0.8, value=0.85}
         <at 1 0.8 RCIM_Cubic>
        ]}
       ZCurve {keys [<at 0 0>,<at 0.0423 0>,<at 1 0>]}
      }}))
   &ScaleColor(
    Scale Alpha=&FloatFromCurve(FloatCurve=DataInterface() {propertyValues {Curve {keys [<at 0.7 1 RCIM_Cubic>,<at 1 0 RCIM_Cubic>]}}})
    Scale RGB=&VectorFromFloat(Value=&FloatFromCurve(FloatCurve=DataInterface() {propertyValues {Curve {keys [<at 0.1 25 RCIM_Cubic>, {arriveTangent=-26.6667, interpMode=RCIM_Cubic, leaveTangent=-26.6667, time=0.2, value=8}, <at 1 1 RCIM_Cubic>]}}}))
    ScaleA=true
    ScaleRGB=true
   )
  ]
  rendererClasses [&NiagaraMeshRendererProperties]
  renderers [{rendererClass=&NiagaraMeshRendererProperties, rendererIndex=0}]
 }
 Sparklies {
  emitterUpdateScript [
   &EmitterState(Life Cycle Mode=::NewEnumerator1, Loop Behavior=::NewEnumerator1, Loop Delay=0.1, Loop Duration=0.05, UseLoopDelay=true)
   &SpawnRate(SpawnRate=200, Use Spawn Probability=false)
  ]
  particleSpawnScript [
   &InitializeParticle(
    Color=<rgba 0.9271 1.9872 2 1>
    Lifetime=&UniformRangedFloat(Minimum=0.15, Maximum=0.85)
    Module.Write Color=true
    Module.Write Lifetime=true
    Module.Write Mass=true
    Module.Write Material Random=true
    Module.Write Position=false
    Module.Write Scale=false
    Module.Write SpriteRotation=true
    Module.Write SpriteSize=true
    Position=&SimulationPosition()
    Sprite Rotation=&RandomRangeFloat(Minimum=0, Maximum=360)
    Sprite Size=&Vector2DFromFloat(Value=&UniformRangedFloat(Minimum=10, Maximum=25))
   )
   &AddVelocity(Velocity=&UniformRangedVector(Minimum=<-25 -25 -50>, Maximum=<25 25 50>))
  ]
  particleUpdateScript [
   &ParticleState()
   &SolveForcesAndVelocity(Clamp Velocity=false, Limit Acceleration=false)
   &CameraOffset(Camera Offset Amount=&FloatFromCurve(FloatCurve=DataInterface() {propertyValues {Curve {keys [<at 0 -1>,<at 1 1>]}}}, Scale Curve=25))
   &ScaleColor(
    Scale RGB=&VectorFromFloat(Value=&FloatFromCurve(FloatCurve=DataInterface() {propertyValues {Curve {keys [{arriveTangent=-7.4351, interpMode=RCIM_Cubic, leaveTangent=-7.4351, tangentMode=RCTM_User, time=0, value=2}, <at 0.4 1 RCIM_Cubic>]}}}))
    ScaleA=true
    ScaleRGB=true
   )
   &SpriteRotationRate(Rotation Rate=&UniformRangedFloat(Minimum=-1, Maximum=1))
   &ScaleSpriteSize(Scale Factor=&Vector2DFromFloat(Value=&FloatFromCurve(FloatCurve=DataInterface() {propertyValues {Curve {keys [
          <at 0 1 RCIM_Cubic>
          {arriveTangent=-0.1702, interpMode=RCIM_Cubic, leaveTangent=-0.1702, time=0.1624, value=0.4936}
          {arriveTangent=-0.5893, interpMode=RCIM_Cubic, leaveTangent=-0.5893, time=0.1739, value=0.9704}
          {arriveTangent=-3.935, interpMode=RCIM_Cubic, leaveTangent=-3.935, time=0.1883, value=0.4783}
          {arriveTangent=0.2826, interpMode=RCIM_Cubic, leaveTangent=0.2826, tangentMode=RCTM_User, time=0.3478, value=0.2863}
          {arriveTangent=-0.4389, interpMode=RCIM_Cubic, leaveTangent=-0.4389, time=0.3697, value=0.8087}
          {arriveTangent=-2.3135, interpMode=RCIM_Cubic, leaveTangent=-2.3135, time=0.395, value=0.2655}
          {arriveTangent=1.3513, interpMode=RCIM_Cubic, leaveTangent=1.3513, time=0.6535, value=0.1521}
          <at 0.6742 0.6428>
          {arriveTangent=-1.1855, interpMode=RCIM_Cubic, leaveTangent=-1.1855, tangentMode=RCTM_User, time=0.6972, value=0.1329}
          <at 1 0 RCIM_Cubic>
         ]}}})))
  ]
  rendererClasses [&NiagaraSpriteRendererProperties]
  renderers [{rendererClass=&NiagaraSpriteRendererProperties}]
 }
}
summary {
 systemName=NS_RacerZero_Swing
 emitters [
  {emitterName=DarkBKGD, rendererClasses [&NiagaraRibbonRendererProperties]}
  {emitterName=Fragments, rendererClasses [&NiagaraMeshRendererProperties]}
  {emitterName=Sparklies, rendererClasses [&NiagaraSpriteRendererProperties]}
 ]
 userVariables []
}
